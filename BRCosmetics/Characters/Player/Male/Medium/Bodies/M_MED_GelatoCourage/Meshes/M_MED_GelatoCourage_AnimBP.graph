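# Animation Blueprint for the GelatoCourage male medium body cosmetic, layering shoulder-setup corrections and a Control Rig onto the character's pose.
# It copies the attached parent's pose, then chains weighted multi-bone constraints placing five dyn_shouldersetup bones before running the Control Rig.
# The update event converts weapon animation type, movement and crouch states into float weights for the rig, alongside skydive, emote and front-end flags.
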AnimBlueprint /BRCosmetics/Characters/Player/Male/Medium/Bodies/M_MED_GelatoCourage/Meshes/M_MED_GelatoCourage_AnimBP
target_skeleton=&/M_MED_GelatoCourage_Skeleton
graphs [
 AnimGraph {nodes [
   "Constraint - dyn_shouldersetup_handle (PPPPPPPP)"!(
    ComponentPose
    Alpha=1
    ConstraintWeights_0=0.0103
    ConstraintWeights_1:float=0.0026
    ConstraintWeights_2:float=0.7708
    ConstraintWeights_3:float=0.0171
    ConstraintWeights_4:float=0.0108
    ConstraintWeights_5:float=0.0548
    ConstraintWeights_6:float=0.121
    ConstraintWeights_7:float=0.0126
   ) -> (Pose) {
    alphaBoolBlend {blendOption=Linear}
    boneToModify {boneName=dyn_shouldersetup_handle}
    constraintSetup [
     {targetBone {boneName=spine_03}}
     {targetBone {boneName=spine_04}}
     {targetBone {boneName=spine_05}}
     {targetBone {boneName=clavicle_l}}
     {targetBone {boneName=clavicle_r}}
     {targetBone {boneName=neck_01}}
     {targetBone {boneName=neck_02}}
     {targetBone {boneName=head}}
    ]
    constraintWeights [1,1,1,1,1,1,1,1]
   }
   "Constraint - dyn_shouldersetup_handle (PPPPP)"!(
    ComponentPose
    Alpha=1
    ConstraintWeights_0=0.6397
    ConstraintWeights_1:float=0.0974
    ConstraintWeights_2:float=0.217
    ConstraintWeights_3:float=0.0457
    ConstraintWeights_4:float=0.0002
   ) -> (Pose) {
    alphaBoolBlend {blendOption=Linear}
    boneToModify {boneName=dyn_shouldersetup_handle}
    constraintSetup [
     {targetBone {boneName=spine_05}}
     {targetBone {boneName=neck_01}}
     {targetBone {boneName=neck_02}}
     {targetBone {boneName=head}}
     {targetBone {boneName=clavicle_l}}
    ]
    constraintWeights [1,1,1,1,1]
   }
   Control Rig!(
    Source
    Alpha
    Sky Dive_Floating:bool=Get bIsSkydiveFloating()
    Sky Dive Paraschute:bool=Get bIsParachuteOpened()
    Sky Dive Diving:bool=Get bIsSkydiveDiveMode()
   ) -> (Pose) {
    alphaBoolBlend {blendOption=Linear}
    controlRigAssetReference {blueprintRigClass=&/M_Med_GelatoCourage_ControlRig.M_Med_GelatoCourage_ControlRig_C}
   }
   n1 = "Constraint - dyn_shouldersetup_handle (PPPPPPPP)"(
    ComponentPose=Local To Component(LocalPose=Copy Pose From Mesh() {bUseAttachedParent=true})
    Alpha=1
    ConstraintWeights_0=0.0413
    ConstraintWeights_1=0.0106
    ConstraintWeights_2=0.8334
    ConstraintWeights_3=0.0205
    ConstraintWeights_4=0.0448
    ConstraintWeights_5=0.0007
    ConstraintWeights_6=0.0284
    ConstraintWeights_7=0.0203
   ) {
    alphaBoolBlend {blendOption=Linear}
    boneToModify {boneName=dyn_shouldersetup_handle}
    constraintSetup [
     {targetBone {boneName=spine_03}}
     {targetBone {boneName=spine_04}}
     {targetBone {boneName=spine_05}}
     {targetBone {boneName=neck_01}}
     {targetBone {boneName=neck_02}}
     {targetBone {boneName=head}}
     {targetBone {boneName=clavicle_l}}
     {targetBone {boneName=clavicle_r}}
    ]
    constraintWeights [1,1,1,1,1,1,1,1]
   }
   n2 = "Constraint - dyn_shouldersetup_l_back (PPPPPPPP)"!(
    ComponentPose=@n1
    Alpha=1
    ConstraintWeights_0=0.0148
    ConstraintWeights_1:float=0.004
    ConstraintWeights_2:float=0.4323
    ConstraintWeights_3:float=0.0004
    ConstraintWeights_4:float=0.0011
    ConstraintWeights_5:float=0.494
    ConstraintWeights_6:float=0.0227
    ConstraintWeights_7:float=0.0307
   ) -> (Pose) {
    alphaBoolBlend {blendOption=Linear}
    boneToModify {boneName=dyn_shouldersetup_l_back}
    constraintSetup [
     {targetBone {boneName=spine_03}}
     {targetBone {boneName=spine_04}}
     {targetBone {boneName=spine_05}}
     {targetBone {boneName=neck_01}}
     {targetBone {boneName=neck_02}}
     {targetBone {boneName=clavicle_l}}
     {targetBone {boneName=upperarm_twist_01_l}}
     {targetBone {boneName=deform_upperarm_up_l}}
    ]
    constraintWeights [1,1,1,1,1,1,1,1]
   }
   n3 = "Constraint - dyn_shouldersetup_l_front (PPPPPPPP)"!(
    ComponentPose=@n2
    Alpha=1
    ConstraintWeights_0=0.0436
    ConstraintWeights_1:float=0.106
    ConstraintWeights_2:float=0.3799
    ConstraintWeights_3:float=0.2983
    ConstraintWeights_4:float=0.0294
    ConstraintWeights_5:float=0.0256
    ConstraintWeights_6:float=0.0209
    ConstraintWeights_7:float=0.0964
   ) -> (Pose) {
    alphaBoolBlend {blendOption=Linear}
    boneToModify {boneName=dyn_shouldersetup_l_front}
    constraintSetup [
     {targetBone {boneName=spine_03}}
     {targetBone {boneName=spine_04}}
     {targetBone {boneName=spine_05}}
     {targetBone {boneName=clavicle_l}}
     {targetBone {boneName=upperarm_twist_01_l}}
     {targetBone {boneName=deform_upperarm_fr_l}}
     {targetBone {boneName=deform_upperarm_up_l}}
     {targetBone {boneName=deform_pec_l}}
    ]
    constraintWeights [1,1,1,1,1,1,1,1]
   }
   n4 = "Constraint - dyn_shouldersetup_r_back (PPPPPPPP)"!(
    ComponentPose=@n3
    Alpha=1
    ConstraintWeights_0=0.0233
    ConstraintWeights_1:float=0.0107
    ConstraintWeights_2:float=0.3621
    ConstraintWeights_3:float=0.4988
    ConstraintWeights_4:float=0.0419
    ConstraintWeights_5:float=0.0022
    ConstraintWeights_6:float=0.0602
    ConstraintWeights_7:float=0.0008
   ) -> (Pose) {
    alphaBoolBlend {blendOption=Linear}
    boneToModify {boneName=dyn_shouldersetup_r_back}
    constraintSetup [
     {targetBone {boneName=spine_03}}
     {targetBone {boneName=spine_04}}
     {targetBone {boneName=spine_05}}
     {targetBone {boneName=clavicle_r}}
     {targetBone {boneName=upperarm_twist_01_r}}
     {targetBone {boneName=deform_upperarm_fr_r}}
     {targetBone {boneName=deform_upperarm_up_r}}
     {targetBone {boneName=deform_pec_r}}
    ]
    constraintWeights [1,1,1,1,1,1,1,1]
   }
   n5 = "Constraint - dyn_shouldersetup_r_front (PPPPPPPP)"!(
    ComponentPose=@n4
    Alpha=1
    ConstraintWeights_0=0.0463
    ConstraintWeights_1:float=0.1073
    ConstraintWeights_2:float=0.2889
    ConstraintWeights_3:float=0.2902
    ConstraintWeights_4:float=0.052
    ConstraintWeights_5:float=0.0496
    ConstraintWeights_6:float=0.0389
    ConstraintWeights_7:float=0.1268
   ) -> (Pose) {
    alphaBoolBlend {blendOption=Linear}
    boneToModify {boneName=dyn_shouldersetup_r_front}
    constraintSetup [
     {targetBone {boneName=spine_03}}
     {targetBone {boneName=spine_04}}
     {targetBone {boneName=spine_05}}
     {targetBone {boneName=clavicle_r}}
     {targetBone {boneName=upperarm_twist_01_r}}
     {targetBone {boneName=deform_upperarm_fr_r}}
     {targetBone {boneName=deform_upperarm_up_r}}
     {targetBone {boneName=deform_pec_r}}
    ]
    constraintWeights [1,1,1,1,1,1,1,1]
   }
   n6 = Control Rig!(
    Source=Component To Local(ComponentPose=@n5)
    Alpha
    Sky Dive_Floating=Get bIsSkydiveFloating()
    Sky Dive Paraschute=Get bIsParachuteOpened()
    Sky Dive Diving=Get bIsSkydiveDiveMode()
    Get Rifle:double=Get Is Aiming Rifle() -> (Is Aiming Rifle:double)
    Get Pistol:double=Get Is Aiming Pistol() -> (Is Aiming Pistol:double)
    Pickaxe:double=Get Is Pickaxe() -> (Is Pickaxe:double)
    Crouching:bool=Get bIsCrouching()
    Sprinting:bool=Get bIsMoving2D()
    Emote:bool=Get bIsPlayingEmote()
    Moving Crouch:double=Get Moving 2d Crouch() -> (Moving 2d Crouch:double)
    Not Moving Crouch:double=Get Not moving 2d Crouch() -> (Not moving 2d Crouch:double)
    Not Crouch:double=Get Not Crouch() -> (Not Crouch:double)
    "Moving ":double="Get Moving 2d "() -> ("Moving 2d ":double)
    Not Moving:double=Get Not moving 2d() -> (Not moving 2d:double)
    Looking Down:double=Get Looking Down() -> (Looking Down:double)
    Front End:bool=Get bIsInFrontEnd()
   ) -> (Pose) {
    alphaBoolBlend {blendOption=Linear}
    controlRigAssetReference {blueprintRigClass=&M_Med_GelatoCourage_ControlRig_C}
   }
   Output Pose(Result=@n6) {layerGroup=None}
  ]}
 EventGraph {nodes [
   Event BlueprintUpdateAnimation()
   > n1 = Sequence!(execute) -> (then_0, then_1, then_2:Exec, then_3:Exec, then_4:Exec, then_5:Exec, then_6:Exec)
   > Gravity Override State Function DEPRECATED(CharacterPartAnimInstance=Self-Reference(), GameTimeInSeconds=GetGameTimeInSeconds(), GravityOverrideParams=Get Gravity Override Params())
   > Set Gravity Override(Gravity Override="0, 0, 0")
   n2 = Cast To FortPlayerAnimInstance(execute=@n1.then_1, Object=GetAnimInstance(self=GetOwnerSkeletalMeshComponent()))
   > Set Is Aiming Pistol(Is Aiming Pistol:double=0) -> (Output_Get:double)
   > Set Is Pickaxe(Is Pickaxe:double=0) -> (Output_Get:double)
   > Set Is Aiming Rifle(Is Aiming Rifle:double=0) -> (Output_Get:double)
   > Set Is Pickaxe Boo(Is Pickaxe Boo:bool=false) -> (Output_Get:bool)
   > n3 = Switch on EFortWeaponCoreAnimation!(execute, Selection:EFortWeaponCoreAnimation Enum=Get WeaponCoreAnim(self=@n2)) -> (
    Melee:Exec
    Pistol:Exec
    Shotgun:Exec
    PaperBlueprint:Exec
    Rifle:Exec
    MeleeOneHand:Exec
    MachinePistol:Exec
    RocketLauncher:Exec
    GrenadeLauncher:Exec
    GoingCommando:Exec
    AssaultRifle:Exec
    TacticalShotgun:Exec
    SniperRifle:Exec
    TrapPlacement:Exec
    ShoulderLauncher:Exec
    AbilityDecoTool:Exec
    Crossbow:Exec
    C4:Exec
    RemoteControl:Exec
    DualWield:Exec
    AR_BullPup:Exec
    AR_ForwardGrip:Exec
    MedPackPaddles:Exec
    SMG_P90:Exec
    AR_DrumGun:Exec
    Consumable_Small:Exec
    Consumable_Large:Exec
    Balloon:Exec
    MountedTurret:Exec
    CreativeTool:Exec
    ExplosiveBow:Exec
    AshtonIndigo:Exec
    AshtonChicago:Exec
    MeleeDualWield:Exec
    Unarmed:Exec
   )
   Branch(execute=[@n3.AssaultRifle, @n3.Crossbow, @n3.GrenadeLauncher, @n3.Rifle, @n3.RocketLauncher, @n3.Shotgun, @n3.SniperRifle, @n3.TacticalShotgun])
   > Set Is Aiming Rifle(Is Aiming Rifle=1)
   Branch(execute=[@n3.MachinePistol, @n3.Pistol])
   > Set Is Aiming Pistol(Is Aiming Pistol=1)
   Branch(execute=@n3.Melee)
   > Set Is Pickaxe(Is Pickaxe=1)
   > Set Is Pickaxe Boo(Is Pickaxe Boo=true)
   n4 = Get bIsMoving2D()
   n5 = Get bIsCrouching()
   Set Moving 2d Crouch(execute=@n1.then_2, Moving 2d Crouch:double="To Float (Boolean)"(InBool=AND Boolean(A=@n4, B=@n5))) -> (Output_Get:double)
   n6 = AND Boolean(A=Get bIsMoving2D(), B=Get bIsCrouching())
   Set Not moving 2d Crouch(execute=@n1.then_3, Not moving 2d Crouch:double="To Float (Boolean)"(InBool=NOT Boolean(A=@n6))) -> (Output_Get:double)
   n7 = Get bIsCrouching()
   Set Not Crouch(execute=@n1.then_4, Not Crouch:double="To Float (Boolean)"(InBool=NOT Boolean(A=@n7))) -> (Output_Get:double)
   "Set Moving 2d "(execute=@n1.then_5, "Moving 2d ":double="To Float (Boolean)"(InBool=Get bIsMoving2D())) -> (Output_Get:double)
   n8 = Get bIsMoving2D()
   Set Not moving 2d(execute=@n1.then_6, Not moving 2d:double="To Float (Boolean)"(InBool=NOT Boolean(A=@n8))) -> (Output_Get:double)
  ]}
]
variables [
 Gravity Override {type=Vector}
 Gravity Override Params {type=GravityOverrideParamsStruct}
 Is Aiming Pistol {type=number}
 Is Aiming Rifle {type=number}
 Is Pickaxe {type=number}
 Is Pickaxe Boo {type=boolean}
 Is playing emote {type=boolean}
 Looking Down {type=number}
 "Moving 2d " {type=number}
 Moving 2d Crouch {type=number}
 Not Crouch {type=number}
 Not moving 2d {type=number}
 Not moving 2d Crouch {type=number}
]
